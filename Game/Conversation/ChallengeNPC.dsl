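; Conversation task node that registers the talking player as the challenger of the NPC in the conversation.
; On execution it casts the NPC participant to BP_PlayerPawn_NonParticipant and resolves the player participant's controller to its controlled FortPlayerPawn, which it passes to SetChallenger.
; It then returns the result of the parent task's execution.
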(fn ExecuteTaskNode
  (params (Context :type Conversation%20Context%20Structure%20%28by%20ref%29))
  (bind _returnvalue (Class|FortConversationContextLibrary|GetParticipantActorNPC Context))
  (bind _returnvalue_2 (Class|FortConversationContextLibrary|GetParticipantActorPlayer Context))
  (bind _asfort_player_controller (Utilities|Casting|CastToFortPlayerController _returnvalue_2))
  (bind _asbp_player_pawn_non_participant (Utilities|Casting|CastToBP_PlayerPawn_NonParticipant _returnvalue)
    (:then
      (bind _asfort_player_pawn (Utilities|Casting|CastToFortPlayerPawn (Pawn|GetControlledPawn _asfort_player_controller)))
      (Class|BPPlayerPawnNonParticipant|SetChallenger _asbp_player_pawn_non_participant _asfort_player_pawn :!class /Game/Athena/AI/NPCs/Base/BP_PlayerPawn_NonParticipant.BP_PlayerPawn_NonParticipant_C))
    (:CastFailed
      _))
  (bind _returnvalue_3 (|CallParentFunction Context))
  (return _returnvalue_3))
