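; Non-player conversation participant component that applies Creative Conversation Device options to its interaction UI.
; SetUIVariables sets indicator bubble, draw distance, text and name visibility, max-conversation text color and a valid marker icon, then broadcasts the settings change.
; Helper functions set the UI relative location, display name and conversation-limit name overrides.
; Both name helpers log and clear their override when given an empty string.
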(event EventBeginPlay)

(event EventTick
  (params (DeltaSeconds :type Float%20%28single-precision%29)))

(fn SetUIVariables
  (params (Show%20Indicator%20Bubble :type Boolean) (Indicator%20Bubble%20Range :type Float%20%28double-precision%29) (Show%20Conversation%20Text :type Boolean) (Show%20Name%20In%20%20Range :type Boolean) (Name :type String) (UI%20Relative%20Location :type Vector) (IconTexture. :type Texture%202D%20Soft%20Object%20Reference) (MaxConversationName :type String) (MaxConversationTextColor :type Slate%20Color%20Structure))
  ;[0.4362 1 0.1726 0.403] Set Creative User Options For the Conversation Device.
  (Variables|FortNonPlayerConversationParticipantComponent|SetbShowIndicatorBubble Show%20Indicator%20Bubble)
  (Variables|InteractionUI|SetMaxDrawDistanceOverride Indicator%20Bubble%20Range)
  (Variables|FortNonPlayerConversationParticipantComponent|SetbShowConversationText Show%20Conversation%20Text)
  (Variables|FortNonPlayerConversationParticipantComponent|SetbShowNameWhenInRange Show%20Name%20In%20%20Range)
  (CallFunction|UpdateNameVars Name)
  ;[0.04466 1 0.917] Set values for Creative Conversation that will show that the conversation max limit was reached.
  ;
  ; The icon is set from the Conversation Device.
  (CallFunction|UpdateMaxConvesrationNameVar MaxConversationName)
  (Variables|CreativeUI|SetCreativeConversationMaxTextColor MaxConversationTextColor)
  (CallFunction|UpdateLocationVars UI%20Relative%20Location)
  (if (Utilities|IsValidSoftObjectReference IconTexture.)
    (Variables|InteractionUI|SetDialogMarkerIndicatorSoftIcon IconTexture.)
    (else
      _))
  (CallFunction|BroadcastConversationSettingsChanged))

(fn UpdateLocationVars
  (params (UI%20Relative%20Location :type Vector))
  (Variables|InteractionUI|SetRelativeLocationOverride UI%20Relative%20Location))

(fn UpdateMaxConvesrationNameVar
  (params (Name :type String))
  ;[0.4162 1 0.849] Update Max ConversationNameOverride
  (if (not (Utilities|String|IsEmpty Name))
    (Variables|CreativeUI|SetParticipantConversationMaxNameOverride (Utilities|Text|ToText(String) Name))
    (else
      (Fortnite|FortShippingLog self "NonPlayerConversationComponent: MaxConversationName Var is Empty")
      (Variables|CreativeUI|SetParticipantConversationMaxNameOverride ""))))

(fn UpdateNameVars
  (params (Name :type String))
  ;[1 0 0] We need to find a way to feed a name in and adjust what the range is for show name in range - minor engineering work.
  (if (not (Utilities|String|IsEmpty Name))
    (Variables|InteractionUI|SetParticipantDisplayNameOverride (Utilities|Text|ToText(String) Name))
    (else
      (Fortnite|FortShippingLog self "NonPlayerConversationComponent: Name Var is Empty")
      (Variables|InteractionUI|SetParticipantDisplayNameOverride ""))))
