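; A conversation task node that shows the user a data-driven reason why something cannot be used.
; ExecuteTaskNode runs the parent logic, then advances the conversation if the parent's result allows it to continue, otherwise returning that result unchanged.
; GatherStaticExtraData appends the CannotUseReason variable to the parent's extra data as a "cannotusereason" parameter pair.
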(fn ExecuteTaskNode
  (params (Context :type Conversation%20Context%20Structure%20%28by%20ref%29))
  (bind _returnvalue (|CallParentFunction Context))
  (bind _returnvalue_2 (Conversation|CanConversationContinue _returnvalue))
  (if _returnvalue_2
    (bind _returnvalue_3 (Conversation|AdvanceConversation Context))
    (return _returnvalue_3)
    (else
      (return _returnvalue))))

(fn GatherStaticExtraData
  (params (Context :type Conversation%20Context%20Structure%20%28by%20ref%29))
  (bind _inoutextradata (|CallParentFunction Context))
  (Utilities|Array|Add _inoutextradata (Utilities|Struct|MakeConversationNodeParameterPair "cannotusereason" (Variables|Default|GetCannotUseReason)))
  (return _inoutextradata))
