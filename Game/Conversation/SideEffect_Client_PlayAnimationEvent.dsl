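; Conversation side effect that plays an animation event on the NPC participant, client-side.
; ClientCauseSideEffect gets the NPC actor from the conversation context and casts it to FortPawn.
; It builds an anim input event from the configurable AnimationEventName variable and triggers it on that pawn.
; A failed cast does nothing.
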(fn ClientCauseSideEffect
  (params (Context :type Conversation%20Context%20Structure%20%28by%20ref%29))
  (bind _returnvalue (Class|FortConversationContextLibrary|GetParticipantActorNPC Context))
  (bind _asfort_pawn (Utilities|Casting|CastToFortPawn _returnvalue)
    (:then
      (bind _returnvalue_2 (Class|FortAnimInputEvent|CreateAnimInputEvent (Variables|Default|GetAnimationEventName)))
      (Animation|TriggerAnimInputEvent _asfort_pawn _returnvalue_2))
    (:CastFailed
      _))
  (return))
