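; Conversation client side effect that plays a configured sound cue on the NPC participant.
; It gets the NPC actor from the conversation context and takes an audio component from those matching a configurable component tag.
; If that component and the actor's NPC_VoiceComponent are both valid, it calls PlaySoundCue on the voice component with the cue and audio component; otherwise it does nothing.
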(fn ClientCauseSideEffect
  (params (Context :type Conversation%20Context%20Structure%20%28by%20ref%29))
  (bind _returnvalue (Class|FortConversationContextLibrary|GetParticipantActorNPC Context))
  (bind _output_get (LocalVariables|SetAudioComponent (Utilities|Array|Get(acopy) (Actor|GetComponentsbyTag _returnvalue "/Script/Engine.AudioComponent" (Variables|Default|GetAudioComponentTag)))))
  (if (and (Utilities|IsValid _output_get :!class /Script/Engine.KismetSystemLibrary) (Utilities|IsValid (Actor|GetComponentbyClass _returnvalue "/Game/Sounds/Fort_Audio_Blueprints/VoiceComponent/NPC_VoiceComponent.NPC_VoiceComponent_C") :!class /Script/Engine.KismetSystemLibrary))
    (Class|NPCVoiceComponent|PlaySoundCue (Actor|GetComponentbyClass _returnvalue "/Game/Sounds/Fort_Audio_Blueprints/VoiceComponent/NPC_VoiceComponent.NPC_VoiceComponent_C") (Variables|Default|GetSoundCueToPlay) (LocalVariables|GetAudioComponent))
    (return)
    (else
      (return))))
