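; Client-side conversation side effect that plays a voice-over line on the NPC participant.
; It gets the NPC actor from the conversation context and casts it to BP_PlayerPawn_NonParticipant.
; When the configured VoiceOverTag gameplay tag is valid, it calls PlayVoiceComponentSoundbyTag on that pawn, otherwise it does nothing.
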(fn ClientCauseSideEffect
  (params (Context :type Conversation%20Context%20Structure%20%28by%20ref%29))
  (bind _returnvalue (Class|FortConversationContextLibrary|GetParticipantActorNPC Context))
  (bind _asbp_player_pawn_non_participant (Utilities|Casting|CastToBP_PlayerPawn_NonParticipant _returnvalue))
  (if (GameplayTags|IsGameplayTagValid (Variables|Default|GetVoiceOverTag))
    (Class|BPPlayerPawnNonParticipant|PlayVoiceComponentSoundbyTag _asbp_player_pawn_non_participant (Variables|Default|GetVoiceOverTag))
    (return)))
