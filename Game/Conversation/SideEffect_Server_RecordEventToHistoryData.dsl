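; Conversation side effect that records an event in participant history data on the server.
; When the server triggers the side effect, it passes the conversation context and its Event Type variable to Record Conversation Event with Participant.
; The event type is a variable on the side effect, so each use can record a different event.
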(fn ServerCauseSideEffect
  (params (Context :type Conversation%20Context%20Structure%20%28by%20ref%29))
  (Conversation|RecordConversationEventwithParticipant Context (Variables|Default|GetEvent%20Type))
  (return))
